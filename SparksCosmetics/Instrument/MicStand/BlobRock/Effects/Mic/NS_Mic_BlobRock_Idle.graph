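# Idle Niagara effect for the BlobRock mic stand cosmetic, showing sloshing liquid and drips on its skeletal mesh.
# Mic_Liquid drives meshes with a spring pendulum and inherited movement, feeding orientation, smoothed slosh and turbulence to dynamic material parameters.
# Drips spawns falling sprites from the mesh that thin out as owner speed rises and follow time-of-day colour.
# A front-end preview flag switches pendulum pivot, tightness and emissive; DebugBool is disabled.
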NiagaraSystem /SparksCosmetics/Instrument/MicStand/BlobRock/Effects/Mic/NS_Mic_BlobRock_Idle
dependencies {
 usedDataInterfaces [
  &NiagaraDataInterfaceCurve
  &NiagaraDataInterfaceArrayFloat3
  &NiagaraDataInterfaceSkeletalMesh
  &NiagaraDataInterfaceColorCurve
  &NiagaraDataInterfaceArrayFloat
 ]
 usedDynamicInputs [&ColorFromCurve, &MakeFloatFromVector]
 usedModules [
  &self:SystemSpawnScript.SystemScriptSource.SystemScriptGraph.NiagaraNodeAssignment_0.SetVariables_#1
  &SystemState
  &self:SystemSpawnScript.SystemScriptSource.SystemScriptGraph.NiagaraNodeAssignment_2.SetVariables_#2
  &self:SystemSpawnScript.SystemScriptSource.SystemScriptGraph.NiagaraNodeAssignment_1.SetVariables_#3
 ]
}
emitters {
 DebugBool {
  bEnabled=false
  emitterUpdateScript [
   &EmitterState()
   &SpawnBurst_Instantaneous(Use Spawn Probability=false, UseLoopCountLimit=true)
   /SetVariables(Emitter.Emissive Value:NiagaraFloat=&MakeCustomFloatFromBool(Bool=$User.bIsFrontEndPreview, True Float=0.9, False Float=0.22)) {enabled=false, isSetParameters=true}
  ]
  particleSpawnScript [&InitializeParticle(
    Color Mode=::NewEnumerator0
    Position Offset=<0 0 100>
    Position Offset Coordinate Space=::Local
    Sprite Size Mode=::NewEnumerator3
    Uniform Sprite Size=50
    UsePositionOffset=true
    Write Material Random=true
   )]
  particleUpdateScript [
   &ParticleState(Kill Particles When Lifetime Has Elapsed=false)
   &DynamicMaterialParameters(
    Index 0 Param 1=&MakeCustomFloatFromBool(Bool=$User.bIsFrontEndPreview, True Float=1, False Float=0)
    Module.Index0Float0Write=true
    Module.Index0Float1Write=true
    Module.Index0Float2Write=true
    Module.Index0Float3Write=true
   )
  ]
  rendererClasses [&NiagaraSpriteRendererProperties]
  renderers [{rendererClass=&NiagaraSpriteRendererProperties}]
 }
 Drips {
  emitterUpdateScript [
   &EmitterState(Life Cycle Mode=::NewEnumerator1, Loop Delay=1, Loop Duration=4, UseLoopDelay=false)
   &SpawnRate(
    Spawn Probability=0.35
    SpawnRate=&Lerp_Float(
     A=1.75
     B=0
     Alpha=&RemapRange(
      Input Value To Remap=&VectorLength(Vector=$Engine.Owner.Velocity)
      Input Min=0
      Input Max=100
      Output Min=0
      Output Max=1
     )
    )
    Use Spawn Probability=true
   )
  ]
  particleSpawnScript [
   &InitializeParticle(
    Color=<rgba 1.4167 1.7745 2 0.9435>
    Lifetime=&RandomRangeFloat(Minimum=0.5, Maximum=0.8)
    Mass Mode=::NewEnumerator3
    Sprite Size=&Multiply_Vector2D_ByFloat(Float=&RandomRangeFloat(Minimum=9, Maximum=12), Vector2D=<0.1 0.5>)
    Sprite Size Mode=::NewEnumerator4
    UsePositionOffset=false
    Write Material Random=true
   )
   &ShapeLocation(
    Apply To Particle Position=::NewEnumerator1
    MaskSpawnGroup=false
    Non Uniform Scale=<0.75 0.45 1>
    Override Seed=false
    Shape Origin=&VectorFromFloat(Value=0)
    Sphere Radius=1
    UseOwnerScale=true
   ) {module=ShapeLocation001}
  ]
  particleUpdateScript [
   &ParticleState()
   &SkeletalMeshLocation(
    Bone / Socket Index=1
    MaskSpawnGroup=false
    Mesh Sampling Type=::NewEnumerator5
    Orientation Sampling=::NewEnumerator1
    Position Sampling=::NewEnumerator1
    Sampled Position Offset:NiagaraPosition=$StackContext.ShapeLocation001.ShapePosition
    Skeletal Mesh=$System.Skeletal Mesh
    Socket Sampling Mode=::NewEnumerator5
   ) {module=SkeletalMeshLocation001}
   &ScaleSpriteSize(Uniform Curve Sprite Scale=DataInterface() {propertyValues {Curve {keys [
        <at 0 0 RCIM_Cubic>
        {arriveTangent=17.0125, interpMode=RCIM_Cubic, leaveTangent=-0.0394, tangentMode=RCTM_Break, time=0.15, value=1}
        {arriveTangentWeight=0.3483, interpMode=RCIM_Cubic, leaveTangentWeight=0.1006, tangentMode=RCTM_User, time=0.75, value=1}
        <at 1 0.5 RCIM_Cubic>
       ]}}})
   &GravityForce(Gravity=<0 0 -600>)
   &AerodynamicDrag(Aerodynamic Drag=0.1, Use Linear Drag=true, Use Rotational Drag=true) {enabled=false}
   &Drag(Use Linear Drag=true, Use Rotational Drag=true)
   &SolveForcesAndVelocity(Clamp Velocity=false, Limit Acceleration=false)
   &ScaleSpriteSizeBySpeed(Max Scale Factor=<1 3>, Velocity Threshold=2000)
   &AlignSpriteToMeshOrientation(SetAlignment=true, SetFacing=true) {enabled=false}
   &ScaleColor(
    Scale Alpha=&FloatFromCurve(
     FloatCurve=DataInterface() {propertyValues {Curve {keys [
         {arriveTangent=3.3333, interpMode=RCIM_Cubic, time=0.1, value=0}
         {arriveTangent=0.0435, arriveTangentWeight=0.0833, interpMode=RCIM_Cubic, leaveTangent=0.0435, tangentMode=RCTM_User, time=0.3, value=1}
         {arriveTangent=0.0176, arriveTangentWeight=0.1, interpMode=RCIM_Cubic, leaveTangent=0.0176, tangentMode=RCTM_User, time=0.6, value=1}
         <at 1 0 RCIM_Cubic>
        ]}}}
     Scale Curve=&FloatFromCurve(
      FloatCurve=DataInterface() {propertyValues {Curve {keys [<at 0 0.8>,<at 0.2396 1>,<at 0.7097 1>,<at 1 0.8>]}}}
      CurveIndex=Linked() {linkedVariable {name=NPC.FortniteNPC.TimeOfDay_0-1, type=NiagaraFloat}}
     )
    )
    Scale RGB=&MakeVectorFromLinearColorRGB(Color=$System.TOD Color)
    ScaleA=true
    ScaleRGB=true
   )
  ]
  rendererClasses [&NiagaraSpriteRendererProperties]
  renderers [{rendererClass=&NiagaraSpriteRendererProperties}]
 }
 Mic_Liquid {
  emitterUpdateScript [
   &EmitterState(
    Inactive Response=::NewEnumerator1
    Life Cycle Mode=::NewEnumerator1
    Loop Behavior=::NewEnumerator1
    Loop Duration Mode=::NewEnumerator1
    UseLoopDelay=false
   )
   &SpawnBurst_Instantaneous(Use Spawn Probability=false, UseLoopCountLimit=false)
   /SetVariables(Emitter.Orientation:Quat4f=$Engine.Owner.Rotation) {isSetParameters=true}
   /SetVariables(Emitter.Position:Vector3f=/Niagara/DynamicInputs/Helpers/SimulationPosition()) {isSetParameters=true}
   /SetVariables(
    Emitter.RotationVelocity:Vector3f=&Divide_Vector(
     A:Vector3f=&Subtract_Vector(
      A=&RotateVectorByQuaternion(Vector To Rotate=<0 0 1>, Quaternion=$Emitter.Orientation)
      B=&RotateVectorByQuaternion(Vector To Rotate=<0 0 1>, Quaternion=$Emitter.Previous.Orientation)
     )
     B:Vector3f=&VectorFromFloat(Value=$Engine.DeltaTime)
    )
    Emitter.Velocity:Vector3f=&Divide_Vector(
     A:Vector3f=&Subtract_Vector(A=$Emitter.Position, B=$Emitter.Previous.Position)
     B:Vector3f=&VectorFromFloat(Value=$Engine.DeltaTime)
    )
    Emitter.Viscosity=0.5
   ) {isSetParameters=true}
   /SetVariables(Emitter.InheritSourceMovement:NiagaraFloat=&Lerp_Float(A=0, B=0.9, Alpha=$Emitter.Viscosity)) {isSetParameters=true}
   /SetVariables(Emitter.Emissive Value:NiagaraFloat=&MakeCustomFloatFromBool(
     Bool=$User.bIsFrontEndPreview
     True Float=0.9
     False Float=&FloatFromCurve(
      FloatCurve=DataInterface() {propertyValues {Curve {keys [
          <at 0.12 0.45 RCIM_Cubic>
          {arriveTangent=-0.5282, interpMode=RCIM_Cubic, leaveTangent=-0.5282, time=0.3, value=0.3}
          <at 0.4987 0.25 RCIM_Cubic>
          {arriveTangent=0.4983, interpMode=RCIM_Cubic, leaveTangent=0.4983, time=0.7, value=0.3}
          <at 0.9 0.45 RCIM_Cubic>
         ]}}}
      CurveIndex=Linked() {linkedVariable {name=NPC.FortniteNPC.TimeOfDay_0-1, type=NiagaraFloat}}
     )
    )) {isSetParameters=true}
  ]
  particleSpawnScript [
   &InitializeParticle(
    Color Mode=::NewEnumerator0
    Mesh Scale Mode=::NewEnumerator3
    Position Offset=&Multiply_VectorByFloat(Vector=<0 0 -1>, Float=16)
    UsePositionOffset=true
    Write Material Random=true
   )
   &SkeletalMeshLocation(
    Bone / Socket Sampling Mode=::NewEnumerator5
    MaskSpawnGroup=false
    Mesh Sampling Type=::NewEnumerator6
    Orientation Sampling=::NewEnumerator1
    Position Sampling=::NewEnumerator2
    Sampled Position Offset=&SelectVectorFromArray(Direct Array Index=0, Vector Selection Array=DataInterface() {propertyValues {FloatData [<0 0 0>]}})
    Sampled Position Offset Coordinate Space:ENiagaraSkelMeshTransforms=::NewEnumerator1
    Skeletal Mesh=$System.Skeletal Mesh
   )
  ]
  particleUpdateScript [
   &ParticleState(Kill Particles When Lifetime Has Elapsed=false, Let Infinitely Lived Particles Die When Emitter Deactivates=true)
   &ScaleMeshSize(Scale Factor=&VectorFromFloat(Value=$System.LiquidScaleRemap))
   &SkeletalMeshLocation(
    Bone / Socket Sampling Mode=::NewEnumerator5
    MaskSpawnGroup=false
    Mesh Sampling Type=::NewEnumerator6
    Orientation Sampling=::NewEnumerator1
    Position Sampling=::NewEnumerator2
    Sampled Position Offset=&SelectVectorFromArray(Direct Array Index=0, Vector Selection Array=DataInterface() {propertyValues {FloatData [<0 0 0>]}})
    Sampled Position Offset Coordinate Space:ENiagaraSkelMeshTransforms=::NewEnumerator1
    Skeletal Mesh=$System.Skeletal Mesh
   ) {module=SkeletalMeshLocation001}
   &SkeletalMeshLocation(
    Bone / Socket Index=1
    MaskSpawnGroup=false
    Mesh Sampling Type=::NewEnumerator5
    Orientation Sampling=::NewEnumerator1
    Position Sampling=::NewEnumerator2
    Skeletal Mesh=$System.Skeletal Mesh
    Socket Sampling Mode=::NewEnumerator5
   ) {module=SkeletalMeshLocation002}
   /SetVariables(Particles.Position_Origin:NiagaraPosition=$Engine.Emitter.SimulationPosition) {isSetParameters=true}
   /SetVariables(Particles.Position_offset:NiagaraPosition=$Output.SkeletalMeshLocation002.SampledPositionWithoutOffset) {isSetParameters=true}
   &PendulumSetup(
    Module.Enable Spring Driver=true
    Pendulum Length:NiagaraFloat=&Multiply_Float(A=10, B=$System.LiquidScaleRemap)
    Pendulum Pivot:Vector3f=/Niagara/DynamicInputs/TypeConversions/MakeCustomPositionFromBool(
     Bool:NiagaraBool=$User.bIsFrontEndPreview
     True Position:NiagaraPosition=$Particles.Position_offset
     False Position:NiagaraPosition=$Particles.Position_Origin
    )
    Pendulum Rest Axis:Vector3f=&NormalizeVector(
     Vector To Normalize=&Add_Vector(
      A=<0 0 -1>
      B=&Lerp_Vector(
       A=<0 0 0>
       B=&TransformVector(Vector=<0 0 -1>, Destination Space=::Simulation)
       Alpha=$System.Timeline.Float
      )
     )
     Safe Normalize=true
     Safe Fallback Threshold=0
     Safe Fallback Vector=<0 0 0.001>
    )
    "Spring Driven Constraint (Tightness)":NiagaraFloat=&MakeCustomFloatFromBool(Bool=$User.bIsFrontEndPreview, True Float=5, False Float=2)
   )
   &GravityForce(Gravity=&Lerp_Vector(A=<0 0 -980>, B=<0 0 -19.6>, Alpha=$Emitter.Viscosity))
   &Drag(Drag=&Lerp_Float(A=0, B=5, Alpha=$Emitter.Viscosity), Use Linear Drag=true, Use Rotational Drag=false)
   &InheritSourceMovement(
    Applied Position Scale=&VectorFromFloat(Value=$Emitter.InheritSourceMovement)
    ApplyForce=false
    ApplyPosition=true
    ApplyPositionFalloff=false
    Source Velocity:Vector3f=$Emitter.Velocity
   )
   &SolveForcesAndVelocity(Clamp Velocity=false, Limit Acceleration=false, Speed Limit=200)
   &PendulumConstraint()
   &CalculateAccurateVelocity(Previous Position:NiagaraPosition=&Add_Vector(
     A=Linked() {linkedVariable {name=Particles.Previous.Position, type=NiagaraPosition}}
     B=&Multiply_VectorByFloat(Vector=$Emitter.Velocity, Float=&Multiply_Float(A=$Emitter.InheritSourceMovement, B=$Engine.DeltaTime))
    ))
   /Niagara/Modules/Update/Orientation/V2/OrientMeshToVector(
    Look At Direction:Vector3f=&NormalizeVector(
     Vector To Normalize=&Subtract_Vector(
      A=&MakeCustomVectorFromBool(
       Bool=$User.bIsFrontEndPreview
       True Vector=Linked() {linkedVariable {name=Particles.Position_offset, type=NiagaraPosition}}
       False Vector=Linked() {linkedVariable {name=Particles.Position_Origin, type=NiagaraPosition}}
      )
      B=Linked() {linkedVariable {name=Particles.Position, type=NiagaraPosition}}
     )
     Safe Normalize=true
     Safe Fallback Threshold=0
     Safe Fallback Vector=<0 0 0.001>
    )
    Override Quat=false
    Reference Axis Vector:Vector3f=<0 0 1>
    UseOptionalSideAxis=false
    UseOptionalUpAxis=false
   )
   &InterpolateOverTime(
    Rate Of Change=2
    Target Float Value:NiagaraFloat=&Max_Float(
     A:NiagaraFloat=&RemapRange(
      Input Value To Remap=&DotProduct(
       Vector A=<0 0 -1>
       Vector B=&ReturnMeshOrientationAxis(Mesh Orientation Axis=::ZAxis, Mesh Orientation Quat=$Particles.MeshOrientation)
      )
      Input Min=-1
      Input Max=0
      Output Min=0
      Output Max=1
     )
     B:NiagaraFloat=&RemapRange(
      Input Value To Remap=&Divide_Float(
       A=&Subtract_Float(A=$Particles.DistanceTraveled, B=$Particles.Previous.DistanceTraveled)
       B=$System.LiquidScaleRemap
      )
      Input Min=2
      Input Max=4
      Output Min=0
      Output Max=1
     )
    )
   )
   /SetVariables(Particles.LiquidSpeed:NiagaraFloat=&ClampFloat(Float=$Particles.InterpolateOverTime.InterpolateOverTime_Float.Moving Average)) {isSetParameters=true}
   &Timeline(Play=&SetBoolByFloatComparison(A=$Particles.LiquidSpeed, B=0.333), Play Rate=3, Timeline End=10)
   /SetVariables(Particles.Turbulence:NiagaraFloat=$Particles.Timeline.NormalizedTime) {isSetParameters=true}
   /Niagara/Modules/Update/Utility/IncrementOverTime(Rate of Change:NiagaraFloat=&Lerp_Float(A=1, B=1.25, Alpha=$Particles.LiquidSpeed))
   /SetVariables(
    Particles.DynamicMaterialParameter:Vector4f=&MakeFloat4FromVectorAndFloat(
     "Vector (RGB)"=&ReturnMeshOrientationAxis(Mesh Orientation Axis=::ZAxis, Mesh Orientation Quat=$Particles.MeshOrientation)
     "Float (Alpha)"=&Lerp_Float(A=0, B=0.62, Alpha=$System.Timeline.NormalizedTime)
    )
    Particles.DynamicMaterialParameter1:Vector4f=&MakeFloat4FromVectorAndFloat(
     "Vector (RGB)"=&ReturnMeshOrientationAxis(Mesh Orientation Axis=::XAxis, Mesh Orientation Quat=$Particles.MeshOrientation)
     "Float (Alpha)"=$Particles.IncrementOverTime.Counter
    )
    Particles.DynamicMaterialParameter2:Vector4f=&MakeVector4(X=$Particles.LiquidSpeed, Y=$Particles.Turbulence, Z=0, W=0)
   ) {isSetParameters=true}
   &CameraOffset()
  ]
  rendererClasses [&NiagaraMeshRendererProperties]
  renderers [
   {rendererClass=&NiagaraMeshRendererProperties, rendererIndex=0}
   {rendererClass=&NiagaraMeshRendererProperties, rendererIndex=1}
  ]
 }
}
summary {
 systemName=NS_Mic_BlobRock_Idle
 emitters [
  {emitterName=Mic_Liquid, rendererClasses [&NiagaraMeshRendererProperties]}
  {emitterName=Drips, rendererClasses [&NiagaraSpriteRendererProperties]}
  {bEnabled=false, emitterName=DebugBool, rendererClasses [&NiagaraSpriteRendererProperties]}
 ]
 userVariables [
  User.bIsFrontEndPreview {type=NiagaraBool}
  User.IsFilling {type=NiagaraBool, defaultValue=true}
  User.IsFinished {type=NiagaraBool, defaultValue=true}
 ]
}
